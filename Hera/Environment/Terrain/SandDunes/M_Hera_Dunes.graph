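# M_Hera_Dunes is a Nanite-enabled sand material for dune meshes that blend into the landscape.
# It uses the sand layer from a landscape-matching function, optionally adjusting ridge normals through a vertex-color and base-height mask.
# With runtime virtual textures, it blends in landscape RVT attributes using sampled height; otherwise it blends normals using a local-bounds gradient.
# Static switches add cliff gradient blending and debug previews, while mobile flattens normals slightly.
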Material /Hera/Environment/Terrain/SandDunes/M_Hera_Dunes
graph {
 properties {
  bTangentSpaceNormal=false
  bUseMaterialAttributes=true
  bUsedWithNanite=true
  physMaterial=/Game/FortressPhysicalMaterials/Sand
  pixelDepthOffsetMode=PDOM_Legacy
  usedShadingModels=Default Lit
 }
 nodes [
  n1 = ScalarParameter() {defaultValue=0.25}
  n2 = Add(A=@n1) {constB=0.25}
  n3 = Sine(None=@n2)
  n4 = Cosine(None=@n2)
  n5 = Sine(None=@n1)
  n6 = Cosine(None=@n1)
  AppendVector(A=DotProduct(B=AppendVector(A=@n3, B=@n4)), B=DotProduct(B=AppendVector(A=@n5, B=@n6)))
  n7 = VertexNormalWS()
  n8 = ComponentMask(None=@n7) {g=true, r=true}
  n9 = ScalarParameter() {defaultValue=0.25}
  n10 = Add(A=@n9) {constB=0.25}
  n11 = AppendVector(A=Sine(None=@n10), B=Cosine(None=@n10))
  n12 = AppendVector(A=Sine(None=@n9), B=Cosine(None=@n9))
  AppendVector(A=AppendVector(A=DotProduct(A=@n8, B=@n11), B=DotProduct(A=@n8, B=@n12)), B=ComponentMask(None=@n7) {b=true})
  n13 = NamedRerouteUsage()
  n14 = NamedRerouteDeclaration(None=ComponentMask(None=ObjectBounds()) {b=true}) {name=DuneHeight, nodeColor=<rgba 0 0.2903 1 1>}
  n15 = Multiply(
   A=@n14
   B=ScalarParameter() {defaultValue=1, desc=% of the dune height, group=1_Dunes, parameterName="Dune Base Offset (not RVT)", sortPriority=30}
  )
  n16 = ScalarParameter() {defaultValue=0.1, desc=% of the dune height, group=1_Dunes, parameterName=Dune Base Offset, sortPriority=30}
  n17 = StaticSwitchParameter(True=@n15, False=VirtualTextureFeatureSwitch!(No=@n15, Yes=Multiply(A=@n14, B=@n16))) {desc=Just to preview normal blending on non RVT platforms, group=0_Debug Dunes, parameterName=Preview RVT Off}
  n18 = Constant() {desc=% of the dune height, r=0.85}
  n19 = Saturate(None=Divide(A=Subtract(A=@n13, B=@n17), B=Multiply(A=@n14, B=@n18)))
  Multiply(A=@n19, B=@n19)
  n20 = WorldPosition()
  n21 = NamedRerouteDeclaration(None=Subtract(
    A=@n20.Z
    B=RuntimeVirtualTextureSample!(UVs, World Position) {materialType=WorldHeight, virtualTexture=/Game/Athena/Apollo/Maps/Landscape/VT/RVT_Apollo_Landscape_Height}.WorldHeight
   )) {name=RVT H, nodeColor=<rgba 1 0.559 0 1>}
  NamedRerouteDeclaration(None=@n21) {name=Landscape Height, nodeColor=<rgba 1 0 0.6094 1>}
  n22 = RuntimeVirtualTextureSampleParameter() {
   group=E_RVT
   materialType=BaseColor_Normal_Specular_YCoCg
   parameterName=RVT
   sortPriority=1
   virtualTexture=/Game/Athena/Apollo/Maps/Landscape/VT/RVT_Apollo_Landscape
  }
  n23 = SetMaterialAttributes!(MaterialAttributes, Base Color=@n22.BaseColor, Specular=@n22.BaseColor, Roughness=@n22.BaseColor, Normal=@n22.BaseColor) {attributeSetTypes [
    #1
    #2
    #3
    #4
   ]}
  n24 = MaterialFunctionCall!(Grass_UseCustomSlopeMask, Grass_CustomSlopeMask) {materialFunction=/Hera/Environment/Terrain/Landscape/Materials/Layers/MF_Hera_StaticMeshLandscapeLayerMatch}
  n25 = NamedRerouteDeclaration(None=@n24.SandLayer) {name=SandLayer, nodeColor=<rgba 0 1 0 1>}
  n26 = GetMaterialAttributes(None=@n25) {attributeGetTypes [#4]}
  n27 = Constant2Vector() {g=1, r=1}
  n28 = ScalarParameter() {defaultValue=0.2, group=1_Dunes, parameterName=Ridge Normal Intensity, sortPriority=51}
  n29 = Normalize(VectorInput=Multiply(A=AppendVector(A=@n27, B=@n28), B=VertexNormalWS()))
  n30 = AppendVector(
   A=Add(A=ComponentMask(None=@n26.Normal) {g=true, r=true}, B=ComponentMask(None=@n29) {g=true, r=true})
   B=Multiply(A=ComponentMask(None=@n26.Normal) {b=true}, B=ComponentMask(None=@n29) {b=true})
  )
  n31 = Multiply(A=VertexColor().R, B=ScalarParameter() {defaultValue=1, group=1_Dunes, parameterName=Ridge_ Vertex Mask Value})
  n32 = NamedRerouteUsage()
  n33 = ScalarParameter() {group=1_Dunes, parameterName="[RVT] Dune Base Offset", sortPriority=30}
  n34 = ScalarParameter() {
   defaultValue=300
   desc="The range of this gradient is limited by the Landscape Brush Smoothness since you need some heightmap gradient for this mask to work\r\n"
   group=1_Dunes
   parameterName="[RVT] Dune Base Falloff"
   sortPriority=30
  }
  n35 = Saturate(None=Divide(A=Subtract(A=@n32, B=@n33), B=@n34))
  n36 = WorldPosition() {worldPositionShaderOffset=WPT_ExcludeAllShaderOffsets}
  n37 = ActorPositionWS()
  n38 = MaterialFunctionCall() {materialFunction=&ObjectLocalBounds}
  n39 = Divide(A=Subtract(A=Subtract(A=@n36.XYZ, B=@n37), B=@n38.Local Bounds Minimum), B=@n38.Local Bounds Size)
  n40 = NamedRerouteDeclaration(None=Saturate(None=ComponentMask(None=@n39) {b=true})) {name=Fade From Center To Top, nodeColor=<rgba 0 0.1562 1 1>}
  n41 = ScalarParameter() {defaultValue=-0.75, group=1_Dunes, parameterName="[Not RVT] Dune Base Vertical Offset"}
  n42 = Saturate(None=Multiply(A=Add(A=@n40, B=@n41), B=ScalarParameter() {defaultValue=5, group=1_Dunes, parameterName="[Not RVT] Dune Base Faloff"}))
  n43 = StaticSwitchParameter(True=@n42, False=VirtualTextureFeatureSwitch(No=@n42, Yes=Multiply(A=@n35, B=@n35))) {desc=Just to preview normal blending on non RVT platforms, group=0_Debug Dunes, parameterName="Preview [Not RVT]"}
  n44 = NamedRerouteDeclaration(None=Min(A=Saturate(None=@n31), B=NamedRerouteDeclaration(None=@n43) {name=Dune Base Mask, nodeColor=<rgba 0 0.625 1 1>})) {name=Dune Ridge Mask, nodeColor=<rgba 0 0.3582 1 1>}
  n45 = SetMaterialAttributes!(
   MaterialAttributes=@n26.MaterialAttributes
   Normal=StaticSwitchParameter(True=LinearInterpolate(A=@n26.Normal, B=@n30, Alpha=@n44), False=@n26.Normal) {group=1_Dunes, parameterName=Adjust Ridge Normal Intensity, sortPriority=50}
  ) {attributeSetTypes [#4]}
  n46 = BlendMaterialAttributes(A=@n23, B=@n45, Alpha=NamedRerouteUsage())
  n47 = SetMaterialAttributes!(MaterialAttributes=@n45, Base Color=@n22.BaseColor, Specular=@n22.BaseColor, Roughness=@n22.BaseColor, Normal) {attributeSetTypes [
    #1
    #2
    #3
    #4
   ]}
  n48 = BlendMaterialAttributes(A=@n23, B=@n47, Alpha=NamedRerouteUsage())
  n49 = Constant2Vector() {g=1, r=1}
  n50 = ScalarParameter() {defaultValue=5, group=1_Dunes, parameterName="[Not RVT] Normal Blend Value"}
  n51 = Normalize(VectorInput=Multiply(A=VertexNormalWS(), B=AppendVector(A=@n49, B=@n50)))
  n52 = ComponentMask(None=@n51) {g=true, r=true}
  n53 = GetMaterialAttributes(None=@n45) {attributeGetTypes [#4]}
  n54 = ComponentMask(None=@n53.Normal) {g=true, r=true}
  n55 = Add(A=@n52, B=@n54)
  n56 = Multiply(A=Multiply(A=ComponentMask(None=@n51) {b=true}, B=ComponentMask(None=@n53.Normal) {b=true}), B=Constant() {r=2})
  n57 = SetMaterialAttributes!(
   MaterialAttributes=@n53.MaterialAttributes
   Normal=LinearInterpolate(A=AppendVector(A=@n55, B=@n56), B=@n53.Normal, Alpha=NamedRerouteUsage()) {constAlpha=1}
  ) {attributeSetTypes [#4]}
  n58 = StaticSwitchParameter(
   True=@n57
   False=VirtualTextureFeatureSwitch(No=@n57, Yes=StaticSwitchParameter(True=@n46, False=@n48) {group=1_Dunes, parameterName="[RVT] Force Biom Sand on TOP", sortPriority=31})
  ) {desc=Just to preview normal blending on non RVT platforms, group=0_Debug Dunes, parameterName="Preview [Not RVT]"}
  NamedRerouteDeclaration(None=@n58) {name=Debug, nodeColor=<rgba 0 0 0 1>}
  Normalize(VectorInput=Constant3Vector() {constant=<rgba 0 0 1 0>})
  n59 = GetMaterialAttributes(None=@n24.DesertGrassLayer) {attributeGetTypes [#4]}
  n60 = LinearInterpolate(A=@n59.Normal, B=GetMaterialAttributes(None=@n25) {attributeGetTypes [#4]}.Normal, Alpha=@n44)
  Reroute(None=SetMaterialAttributes(MaterialAttributes=@n59.MaterialAttributes, Normal=@n60) {attributeSetTypes [#4]})
  Reroute(None=@n60)
  n61 = NamedRerouteUsage()
  n62 = SetMaterialAttributes!(MaterialAttributes, Emissive Color=NamedRerouteUsage()) {attributeSetTypes [#5]}
  n63 = Constant()
  n64 = StaticSwitch(
   True=BlendMaterialAttributes(
    A=@n58
    B=MaterialFunctionCall!(MA=@n58, Biom Mask, Spec_G Mask=@n63) {materialFunction=/Game/Athena/Helios/Environments/Landscape/Materials/MaterialFunctions_Utilities/MF_CliffZGradient_2}.MA
    Alpha=NamedRerouteUsage()
   )
   False=@n58
   Value=StaticBoolParameter() {group=1_Dunes, parameterName=Use CliffZGradient, sortPriority=0}
  )
  n65 = GetMaterialAttributes(None=NamedRerouteUsage()) {attributeGetTypes [#4]}
  n66 = Constant3Vector() {constant=<rgba 0 0 1 0>}
  ShadingPathSwitch(
   Default=StaticSwitchParameter(
    True=SetMaterialAttributes(Emissive Color=@n61) {attributeSetTypes [#5]}
    False=StaticSwitchParameter(True=@n62, False=@n64) {desc="Used as \"UseCliffZGradient\" Mask", group=0_Debug Dunes, parameterName=Debug Dune Base Mask}
   ) {desc="Used for: RVT Blend and for \"Adjust Ridge Normal Intensity\" ", group=0_Debug Dunes, parameterName=Debug Dune Ridge Mask}
   Mobile=SetMaterialAttributes!(MaterialAttributes=@n65.MaterialAttributes, Normal=LinearInterpolate(A=@n65.Normal, B=@n66) {constAlpha=0.2}) {attributeSetTypes [#4]}
  )
  n67 = NamedRerouteDeclaration(None=Constant()) {name=Not RVT_H, nodeColor=<rgba 1 0.6797 0 1>}
  StaticSwitchParameter(True=@n67, False=VirtualTextureFeatureSwitch(No=@n67, Yes=@n21)) {desc=Just to preview normal blending on non RVT platforms, group=0_Debug Dunes, parameterName="Preview [Not RVT]"}
  Subtract(A=@n20.Z, B=ComponentMask(None=ObjectPositionWS()) {b=true})
  Subtract(A=@n54, B=@n52)
 ]
}
